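# Physics asset for the female medium Rock Climber player cosmetic, simulating dangling gear such as pouches, carabiner, zipper and rope bundle.
# Box bodies on dyn_ bones hang by angular-limited constraints from pelvis, spine_05 and thigh_l, which carry simple box or capsule shapes.
# The rope bundle is a two-level chain: a wide-swinging parent on the left thigh with top, middle and bottom pieces constrained to it.
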PhysicsAsset /BRCosmetics/Characters/Player/Female/Medium/Bodies/F_MED_RockClimber/Meshes/F_MED_Rock_Climber_Physics
bodies {
 dyn_carabiner_front [{shapeType=Box, extentX=4.8744, extentY=7.4036, extentZ=1.6507, center=<2.0726 0 0>}]
 dyn_pouch_bottom_l [{shapeType=Box, extentX=6.4501, extentY=4, extentZ=2.9299, center=<1.0784 0 0>}]
 dyn_ropebundle_bottom [{shapeType=Box, extentX=6.0975, extentY=13.2537, extentZ=1.2219, center=<3.9077 0 -0.0632>, rotation=<rot -2.1105 0 6.4957>}]
 dyn_ropebundle_mid [{shapeType=Box, extentX=8.5975, extentY=17.7537, extentZ=1.2219, center=<2.0308 -0.0006 -0.0025>, rotation=<rot 0.9377 0.2505 -2.9198>}]
 dyn_ropebundle_parent [{shapeType=Box, extentX=10.0296, extentY=13.3022, extentZ=2.3935, center=<5.252 0 0.2304>, rotation=<rot -1.6641 0 0>}]
 dyn_ropebundle_top [{shapeType=Box, extentX=8.5975, extentY=15.2537, extentZ=1.2219, center=<2.9306 0 -0.0138>, rotation=<rot -10.984 -1.1158 6.4918>}]
 dyn_sidepouch_lid_l [{shapeType=Box, extentX=5.6882, extentY=5.349, extentZ=1.3876, center=<2.3943 0 0>}]
 dyn_sidepouch_top_r [{shapeType=Box, extentX=12.4516, extentY=8.6735, extentZ=4.3345, center=<3.8332 0 1.6129>}]
 dyn_smallpouch_back [{shapeType=Box, extentX=4.8745, extentY=4.4462, extentZ=1.3586, center=<1.9143 0 0>}]
 dyn_zipper [{shapeType=Box, extentX=3.26, extentY=3.76, extentZ=0.5, center=<1.7005 0 -0.0001>}]
 pelvis [{shapeType=Box, extentX=7.5, extentY=7.5, extentZ=7.5}]
 spine_05 [{shapeType=Box, extentX=7.5, extentY=7.5, extentZ=7.5, center=<-6.3373 9.5435 -0.0001>, rotation=<rot 0 -38.4301 0>}]
 thigh_l [{shapeType=Capsule, length=22, radius=8, center=<-14.787 0.0322 0.6605>, rotation=<rot 86.5707 125.7717 125.7233>}]
}
constraints [
 {
  bone1Name=dyn_sidepouch_lid_l
  bone2Name=pelvis
  swing1LimitDegrees=3
  swing1Motion=Locked
  swing2LimitDegrees=20
  twistLimitDegrees=3
  twistMotion=Locked
 }
 {bone1Name=dyn_carabiner_front, bone2Name=pelvis, swing1LimitDegrees=45, swing1Motion=Locked, swing2LimitDegrees=30, twistLimitDegrees=3}
 {bone1Name=dyn_ropebundle_parent, bone2Name=thigh_l, swing1LimitDegrees=70, swing2LimitDegrees=60, twistLimitDegrees=85}
 {bone1Name=dyn_ropebundle_top, bone2Name=dyn_ropebundle_parent, swing1LimitDegrees=25, swing2LimitDegrees=25, twistLimitDegrees=5}
 {bone1Name=dyn_ropebundle_mid, bone2Name=dyn_ropebundle_parent, swing1LimitDegrees=25, swing2LimitDegrees=25, twistLimitDegrees=5}
 {bone1Name=dyn_ropebundle_bottom, bone2Name=dyn_ropebundle_parent, swing1LimitDegrees=25, swing2LimitDegrees=25, twistLimitDegrees=5}
 {bone1Name=dyn_zipper, bone2Name=spine_05, swing1LimitDegrees=10, swing2LimitDegrees=15, twistLimitDegrees=3}
 {bone1Name=dyn_smallpouch_back, bone2Name=pelvis, swing1LimitDegrees=10, swing2LimitDegrees=1, twistLimitDegrees=3}
 {bone1Name=dyn_sidepouch_top_r, bone2Name=pelvis, swing1LimitDegrees=15, swing2LimitDegrees=15, twistLimitDegrees=3}
 {bone1Name=dyn_pouch_bottom_l, bone2Name=thigh_l, swing1LimitDegrees=8, swing2LimitDegrees=4, twistLimitDegrees=2}
]
