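# Physics asset for the female medium Rock Climber cosmetic's shoes, set up for dangling shoelace motion on both feet.
# Each shoe has inside and outside lace boxes, each constrained to a small per-side parent box with identical, narrow swing and twist limits.
# Capsules on both lower calf twist bones and a pelvis box add body shapes, with no constraints linking them to the laces.
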PhysicsAsset /BRCosmetics/Characters/Player/Female/Medium/Bodies/F_MED_RockClimber/Meshes/F_MED_Rock_Climber_Shoe_Physics
bodies {
 calf_twist_01_l [{shapeType=Capsule, length=25.75, radius=4.25, center=<-0.1167 -1.6962 0.0037>, rotation=<rot -88.1551 179.9996 176.0628>}]
 calf_twist_01_r [{shapeType=Capsule, length=25.75, radius=4.25, center=<0.1155 1.6788 -0.0038>, rotation=<rot -88.1551 179.9996 176.0628>}]
 dyn_shoelace_l_inside [{shapeType=Box, extentX=4.2022, extentY=1.9566, extentZ=1.1906, center=<2.0244 0 -0.0001>}]
 dyn_shoelace_l_outside [{shapeType=Box, extentX=4.7369, extentY=2.5467, extentZ=0.5, center=<2.1921 0 0>}]
 dyn_shoelace_l_parent [{shapeType=Box, extentX=1, extentY=1, extentZ=1}]
 dyn_shoelace_r_inside [{shapeType=Box, extentX=4.2022, extentY=1.9566, extentZ=1.1906, center=<2.0244 0 -0.0001>}]
 dyn_shoelace_r_outside [{shapeType=Box, extentX=4.2022, extentY=1.9566, extentZ=1.1906, center=<2.0244 0 -0.0001>}]
 dyn_shoelace_r_parent [{shapeType=Box, extentX=1, extentY=1, extentZ=1}]
 pelvis [{shapeType=Box, extentX=7.5, extentY=7.5, extentZ=7.5}]
}
constraints [
 {bone1Name=dyn_shoelace_l_inside, bone2Name=dyn_shoelace_l_parent, swing1LimitDegrees=30, swing2LimitDegrees=20, twistLimitDegrees=5}
 {bone1Name=dyn_shoelace_l_outside, bone2Name=dyn_shoelace_l_parent, swing1LimitDegrees=30, swing2LimitDegrees=20, twistLimitDegrees=5}
 {bone1Name=dyn_shoelace_r_inside, bone2Name=dyn_shoelace_r_parent, swing1LimitDegrees=30, swing2LimitDegrees=20, twistLimitDegrees=5}
 {bone1Name=dyn_shoelace_r_outside, bone2Name=dyn_shoelace_r_parent, swing1LimitDegrees=30, swing2LimitDegrees=20, twistLimitDegrees=5}
]
