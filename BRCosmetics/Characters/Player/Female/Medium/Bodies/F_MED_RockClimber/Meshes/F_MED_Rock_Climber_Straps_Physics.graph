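# Physics asset for the female medium Rock Climber body's straps, covering two dangling rope chains that end in carabiners.
# Both chains hang from a small box body at dyn_ropes_base: three capsule segments on the left, two on the right, each followed by a carabiner capsule.
# Constraints with narrow swing and twist limits link each chain.
# Unconstrained pelvis, right thigh and right calf bodies supply body shapes alongside the chains.
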PhysicsAsset /BRCosmetics/Characters/Player/Female/Medium/Bodies/F_MED_RockClimber/Meshes/F_MED_Rock_Climber_Straps_Physics
bodies {
 calf_r [{shapeType=Box, extentX=16.5, extentY=12.5, extentZ=7.5, center=<-20.9961 0.0609 7.478>, rotation=<rot 0 -5.1474 0>}]
 dyn_carabiner_l [{shapeType=Capsule, length=5.6815, radius=2.2834, center=<4.1056 0.0032 -0.6687>, rotation=<rot -79.8502 179.9995 -179.9995>}]
 dyn_carabiner_r [{shapeType=Capsule, length=5.5047, radius=2.2164, center=<4.3454 0 -0.153>, rotation=<rot 86.629 0.0004 0.0002>}]
 dyn_rope_l_01 [{shapeType=Capsule, length=2.2174, radius=1.0148, center=<1.8204 0 -0.3154>, rotation=<rot 79.1333 -0.0001 -0.0001>}]
 dyn_rope_l_02 [{shapeType=Capsule, length=2.3511, radius=1.1502, center=<1.7326 -0.0001 -0.5879>, rotation=<rot 71.2371 0 0>}]
 dyn_rope_l_03 [{shapeType=Capsule, length=2.1027, radius=1.2275, center=<1.8394 0.2078 -0.1412>, rotation=<rot -86.9325 179.9991 -179.9991>}]
 dyn_rope_r_01 [{shapeType=Capsule, length=3.6344, radius=0.9365, center=<2.6979 0 -0.0828>, rotation=<rot 86.8434 -0.0001 -0.0001>}]
 dyn_rope_r_02 [{shapeType=Capsule, length=3.3844, radius=1.1865, center=<2.415 0 0.1063>, rotation=<rot 89.0799 0.0005 0.0005>}]
 dyn_ropes_base [{shapeType=Box, extentX=1.25, extentY=1.25, extentZ=1.25}]
 pelvis [
  {shapeType=Capsule, length=20, radius=7.75, center=<-4.6927 -1.8133 -0.4492>}
  {shapeType=Capsule, length=20, radius=7.75, center=<0.2795 -0.7244 0.0001>}
  {shapeType=Capsule, length=12.25, radius=5.5, center=<-2.3954 -6.5086 0.991>, rotation=<rot 0 0 3.2539>}
 ]
 thigh_r [{shapeType=Capsule, length=22.75, radius=7.25, center=<16.9695 0.7191 0>, rotation=<rot 90 3.5763 10.3507>}]
}
constraints [
 {bone1Name=dyn_rope_r_01, bone2Name=dyn_ropes_base, swing1LimitDegrees=30, swing2LimitDegrees=30, twistLimitDegrees=10}
 {bone1Name=dyn_rope_r_02, bone2Name=dyn_rope_r_01, swing1LimitDegrees=25, swing2LimitDegrees=25, twistLimitDegrees=10}
 {bone1Name=dyn_carabiner_r, bone2Name=dyn_rope_r_02, swing1LimitDegrees=20, swing2LimitDegrees=20, twistLimitDegrees=10}
 {bone1Name=dyn_rope_l_01, bone2Name=dyn_ropes_base, swing1LimitDegrees=25, swing2LimitDegrees=25, twistLimitDegrees=8}
 {bone1Name=dyn_rope_l_02, bone2Name=dyn_rope_l_01, swing1LimitDegrees=30, swing2LimitDegrees=30, twistLimitDegrees=10}
 {bone1Name=dyn_rope_l_03, bone2Name=dyn_rope_l_02, swing1LimitDegrees=20, swing2LimitDegrees=20, twistLimitDegrees=10}
 {bone1Name=dyn_carabiner_l, bone2Name=dyn_rope_l_03, swing1LimitDegrees=15, swing2LimitDegrees=15, twistLimitDegrees=10}
]
