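# Post-processes the female medium Rock Climber body, adding secondary motion to a pose copied from the attached parent mesh.
# Thigh swing angles drive corrective sub-thigh scaling, weighted constraints position glowsticks and shoulder pads, and AnimDynamics sways the glowsticks.
# Three RigidBody passes simulate separate physics assets, two of them damped while skydiving.
# The event graph also resets dynamics when skydiving ends and applies a delayed upward force to one pass during jumps.
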AnimBlueprint /BRCosmetics/Characters/Player/Female/Medium/Bodies/F_MED_RockClimber/Meshes/F_MED_Rock_Climber_AnimBp
target_skeleton=&/F_MED_Rock_Climber_Skeleton
graphs [
 AnimGraph {nodes [
   n1 = "Transform (Modify) Bone - Bone: dyn_subthigh_l"(
    ComponentPose=Local To Component(LocalPose=Copy Pose From Mesh() {bUseAttachedParent=true})
    Alpha=Get legSwingDelta_l() -> (legSwingDelta_l:double)
    Translation="0,0,0"
    Rotation="0,0,0"
    Scale="2.25,1,2.25"
   ) {
    scaleMode=BMM_Additive
    scaleSpace=BCS_BoneSpace
    alphaBoolBlend {blendOption=Linear}
    boneToModify {boneName=dyn_subthigh_l}
    scale {x=2.25, z=2.25}
   }
   n2 = "Transform (Modify) Bone - Bone: dyn_subthigh_r"(ComponentPose=@n1, Alpha=Get legSwingDelta_r() -> (legSwingDelta_r:double), Translation="0,0,0", Rotation="0,0,0", Scale="2.25,1,2.25") {
    scaleMode=BMM_Additive
    scaleSpace=BCS_BoneSpace
    alphaBoolBlend {blendOption=Linear}
    boneToModify {boneName=dyn_subthigh_r}
    scale {x=2.25, z=2.25}
   }
   n3 = "Constraint - dyn_glowstick_01 (PP)"!(ComponentPose=@n2, Alpha=1, ConstraintWeights_0=0.877, ConstraintWeights_1:float=0.123) -> (Pose) {
    alphaBoolBlend {blendOption=Linear}
    boneToModify {boneName=dyn_glowstick_01}
    constraintSetup [{targetBone {boneName=thigh_r}}, {targetBone {boneName=thigh_twist_01_r}}]
    constraintWeights [1,1]
   }
   n4 = "Constraint - dyn_glowstick_02 (P)"!(ComponentPose=@n3, Alpha=1, ConstraintWeights_0=1, ConstraintWeights_1:float=0) -> (Pose) {
    alphaBoolBlend {blendOption=Linear}
    boneToModify {boneName=dyn_glowstick_02}
    constraintSetup [{targetBone {boneName=thigh_r}}]
    constraintWeights [1,0]
   }
   n5 = Anim Dynamics - Bone: dyn_glowstick_01(ComponentPose=@n4, Alpha=1, ExternalForce="0,0,0") {
    angularDampingOverride=0.99
    bGravityOverrideInSimSpace=true
    bOverrideAngularDamping=true
    bOverrideLinearDamping=true
    linearDampingOverride=0.99
    simulationSpace=RootRelative
    alphaBoolBlend {blendOption=Linear}
    boundBone {boneName=dyn_glowstick_01}
    componentAppliedLinearAccClamp=<400 400 400>
    componentLinearAccScale=<0.25 0.25 0.25>
    componentLinearVelScale=<0.25 0.25 0.25>
    physicsBodyDefinitions [{
      sphereCollisionRadius=0
      boundBone {boneName=dyn_glowstick_01}
      boxExtents=<4 2 2>
      constraintSetup {angularLimitsMax=<0 5 5>, angularLimitsMin=<0 -5 -5>, angularTarget=<0 0 0>}
      localJointOffset=<-1 0 0>
     }]
    relativeSpaceBone {boneName=root}
   }
   n6 = Anim Dynamics - Bone: dyn_glowstick_02(ComponentPose=@n5, Alpha=1, ExternalForce="0,0,0") {
    angularDampingOverride=0.99
    bGravityOverrideInSimSpace=true
    bOverrideAngularDamping=true
    bOverrideLinearDamping=true
    linearDampingOverride=0.99
    simulationSpace=RootRelative
    alphaBoolBlend {blendOption=Linear}
    boundBone {boneName=dyn_glowstick_02}
    componentAppliedLinearAccClamp=<400 400 400>
    componentLinearAccScale=<0.25 0.25 0.25>
    componentLinearVelScale=<0.25 0.25 0.25>
    physicsBodyDefinitions [{
      sphereCollisionRadius=0
      boundBone {boneName=dyn_glowstick_02}
      boxExtents=<4 2 2>
      constraintSetup {angularLimitsMax=<0 5 5>, angularLimitsMin=<0 -5 -5>, angularTarget=<0 0 0>}
      localJointOffset=<-1 0 0>
     }]
    relativeSpaceBone {boneName=root}
   }
   n7 = "Constraint - const_shoulderpad_l (PPPPPPP)"!(
    ComponentPose=@n6
    Alpha=1
    ConstraintWeights_0=0.7195
    ConstraintWeights_1:float=0
    ConstraintWeights_2:float=0.0038
    ConstraintWeights_3:float=0.1594
    ConstraintWeights_4:float=0.1173
    ConstraintWeights_5:float=0.0001
    ConstraintWeights_6:float=0
   ) -> (Pose) {
    alphaBoolBlend {blendOption=Linear}
    boneToModify {boneName=const_shoulderpad_l}
    constraintSetup [
     {targetBone {boneName=clavicle_l}}
     {targetBone {boneName=spine_03}}
     {targetBone {boneName=spine_04}}
     {targetBone {boneName=spine_05}}
     {targetBone {boneName=upperarm_twist_01_l}}
     {targetBone {boneName=clavicle_r}}
     {targetBone {boneName=neck_02}}
    ]
    constraintWeights [1,1,1,1,1,1,1]
   }
   n8 = "Constraint - const_shoulderpad_r (PPPPPPPPPPPPP)"!(
    ComponentPose=@n7
    Alpha=1
    ConstraintWeights_0=0
    ConstraintWeights_1:float=0
    ConstraintWeights_2:float=0.0926
    ConstraintWeights_3:float=0
    ConstraintWeights_4:float=0
    ConstraintWeights_5:float=0.0001
    ConstraintWeights_6:float=0.0388
    ConstraintWeights_7:float=0.2903
    ConstraintWeights_8:float=0
    ConstraintWeights_9:float=0.0003
    ConstraintWeights_10:float=0.5778
    ConstraintWeights_11:float=0
    ConstraintWeights_12:float=0
   ) -> (Pose) {
    alphaBoolBlend {blendOption=Linear}
    boneToModify {boneName=const_shoulderpad_r}
    constraintSetup [
     {targetBone {boneName=head}}
     {targetBone {boneName=clavicle_l}}
     {targetBone {boneName=upperarm_twist_01_r}}
     {targetBone {boneName=spine_01}}
     {targetBone {boneName=spine_02}}
     {targetBone {boneName=spine_03}}
     {targetBone {boneName=spine_04}}
     {targetBone {boneName=spine_05}}
     {targetBone {boneName=upperarm_twist_01_l}}
     {targetBone {boneName=neck_02}}
     {targetBone {boneName=clavicle_r}}
     {targetBone {boneName=pelvis}}
     {targetBone {boneName=neck_01}}
    ]
    constraintWeights [1,1,1,1,1,1,1,1,1,1,1,1,1]
   }
   n9 = RigidBody(ComponentPose=@n8, Alpha=0.75) {
    bClampLinearTranslationLimitToRefPose=true
    bOverrideWorldGravity=true
    overridePhysicsAsset=&/F_MED_Rock_Climber_Shoe_Physics
    simulationSpace=BaseBoneSpace
    alphaBoolBlend {blendOption=Linear}
    baseBoneRef {boneName=root}
    componentAppliedLinearAccClamp=<4000 4000 4000>
    componentLinearAccScale=<0.25 0.25 0.25>
    componentLinearVelScale=<0.25 0.25 0.25>
    overrideWorldGravity {z=-980}
   }
   n10 = RigidBody(ComponentPose=@n9, Alpha=Get Skydive_Alpha() -> (Skydive_Alpha:double), ExternalForce=Get Ext_Force() -> (Ext_Force:Vector)) {
    bClampLinearTranslationLimitToRefPose=true
    bOverrideWorldGravity=true
    overridePhysicsAsset=&/F_MED_Rock_Climber_Physics
    simulationSpace=BaseBoneSpace
    alphaBoolBlend {blendOption=Linear}
    baseBoneRef {boneName=pelvis}
    componentAppliedLinearAccClamp=<4000 4000 4000>
    componentLinearAccScale=<0.25 0.25 0.25>
    componentLinearVelScale=<0.25 0.25 0.25>
    overrideWorldGravity {z=-980}
   }
   n11 = RigidBody(ComponentPose=@n10, Alpha=Get Skydive_Alpha()) {
    bClampLinearTranslationLimitToRefPose=true
    bOverrideWorldGravity=true
    overridePhysicsAsset=&/F_MED_Rock_Climber_Straps_Physics
    simulationSpace=BaseBoneSpace
    alphaBoolBlend {blendOption=Linear}
    baseBoneRef {boneName=root}
    componentAppliedLinearAccClamp=<600 600 600>
    componentLinearAccScale=<0.25 0.25 0.25>
    componentLinearVelScale=<0.25 0.25 0.25>
    overrideWorldGravity {z=-980}
   }
   Output Pose(Result=Component To Local(ComponentPose=@n11)) {layerGroup=None}
  ]}
 EventGraph {nodes [
   Event BlueprintUpdateAnimation()
   > n1 = Sequence!(execute) -> (then_0, then_1, then_2:Exec)
   n2 = Get bIsSkyDiving()
   n3 = Get bIsSkydiveDiveMode()
   > n4 = Set Skydive_Alpha(Skydive_Alpha:double=SelectFloat(A=0.25, B=1, bPickA=OR Boolean(A=@n3, B=@n2))) -> (Output_Get:double)
   Do Once(execute=@n8.then, Reset=@n4.then)
   > ResetDynamics(InTeleportType=ResetPhysics)
   n5 = Branch(execute=@n1.then_2, Condition=Get bIsJumping())
   > Delay(Duration=0.25)
   > Set Ext_Force(Ext_Force:Vector="0,0,1250") -> (Output_Get:Vector)
   > n6 = Delay(Duration=0.375)
   Do Once(execute=@n5.else, Reset=@n6.then)
   > Set Ext_Force(Ext_Force="0, 0, 0")
   n7 = GetOwnerSkeletalMeshComponent()
   Set legSwingDelta_r(
    execute=@n1.then_1
    legSwingDelta_r:double=MapRangeClamped(Value=GetTwistAndSwingAngleOfDeltaRotationFromRefPose(self=@n7, BoneName=thigh_r).OutSwingAngle, InRangeA=150, InRangeB=75)
   ) -> (Output_Get:double)
   > Set legSwingDelta_l(legSwingDelta_l:double=MapRangeClamped(Value=GetTwistAndSwingAngleOfDeltaRotationFromRefPose(self=@n7, BoneName=thigh_l).OutSwingAngle, InRangeA=43, InRangeB=110)) -> (Output_Get:double)
   n8 = Event OnEndSkydiving()
   TryGetPawnOwner()
  ]}
]
variables [
 Ext_Force {type=Vector}
 legSwingDelta_l {type=number}
 legSwingDelta_r {type=number}
 Skydive_Alpha {type=number}
]
