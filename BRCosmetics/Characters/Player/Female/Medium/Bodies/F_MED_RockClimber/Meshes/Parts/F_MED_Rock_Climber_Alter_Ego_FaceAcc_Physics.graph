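# Physics asset for the face accessory mesh of the female Rock Climber Alter Ego character, setting up secondary motion for hair bangs and a hood.
# Capsule bangs bodies hang off a small parent box, with bangs_03 chained to bangs_02, whose second swing axis is locked.
# Three hood boxes each constrain directly to neck_02 with their own swing and twist limits.
# A head capsule and neck box supply body volumes, and the head has no constraint.
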PhysicsAsset /BRCosmetics/Characters/Player/Female/Medium/Bodies/F_MED_RockClimber/Meshes/Parts/F_MED_Rock_Climber_Alter_Ego_FaceAcc_Physics
bodies {
 dyn_bangs_01 [{shapeType=Capsule, length=2.9663, radius=0.5368, center=<1.5811 0 0.0792>, rotation=<rot 90 0 0>}]
 dyn_bangs_02 [{shapeType=Capsule, length=1.8882, radius=0.5993, center=<1.3766 0 0>, rotation=<rot 90 0 0>}]
 dyn_bangs_03 [{shapeType=Capsule, length=2.2074, radius=0.5192, center=<1.2083 0 0>, rotation=<rot 90 0 0>}]
 dyn_bangs_4 [{shapeType=Capsule, length=3.2995, radius=0.7045, center=<1.4267 0 0>, rotation=<rot 90 0 0>}]
 dyn_bangs_parent [{shapeType=Box, extentX=0.75, extentY=0.75, extentZ=0.75}]
 dyn_hood_01 [{shapeType=Box, extentX=4.1165, extentY=1.1907, extentZ=2.538, center=<1.3427 0 -0.3041>}]
 dyn_hood_02 [{shapeType=Box, extentX=3.4104, extentY=1.2429, extentZ=3.0162, center=<0.5998 0 -0.4961>}]
 dyn_hood_03 [{shapeType=Box, extentX=4.1459, extentY=1.6672, extentZ=4.1565, center=<1.2847 0 -1.0591>}]
 head [{shapeType=Capsule, length=5.5, radius=5.75, center=<-2.0306 4.6924 0.6713>, rotation=<rot 87.8363 -180 -180>}]
 neck_02 [{shapeType=Box, extentX=2.5, extentY=2.5, extentZ=2.5}]
}
constraints [
 {bone1Name=dyn_bangs_01, bone2Name=dyn_bangs_parent, swing1LimitDegrees=35, swing2LimitDegrees=20, twistLimitDegrees=12}
 {bone1Name=dyn_bangs_02, bone2Name=dyn_bangs_parent, swing1LimitDegrees=10, swing2LimitDegrees=20, swing2Motion=Locked, twistLimitDegrees=12}
 {bone1Name=dyn_bangs_03, bone2Name=dyn_bangs_02, swing1LimitDegrees=20, swing2LimitDegrees=10, twistLimitDegrees=12}
 {bone1Name=dyn_bangs_4, bone2Name=dyn_bangs_parent, swing1LimitDegrees=25, swing2LimitDegrees=25, twistLimitDegrees=12}
 {bone1Name=dyn_hood_01, bone2Name=neck_02, swing1LimitDegrees=20, swing2LimitDegrees=25, twistLimitDegrees=8}
 {bone1Name=dyn_hood_02, bone2Name=neck_02, swing1LimitDegrees=25, swing2LimitDegrees=15, twistLimitDegrees=45}
 {bone1Name=dyn_hood_03, bone2Name=neck_02, swing1LimitDegrees=35, swing2LimitDegrees=15, twistLimitDegrees=6}
]
