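# Animation Blueprint for the face accessory mesh of the female RockClimber character body, adding secondary physics motion.
# It copies the pose from the attached parent mesh, then runs RigidBody simulation with an override physics asset in spine_05 base-bone space.
# The simulation uses scaled-down component velocity and acceleration and overridden gravity.
# Each update sets Skydive_Alpha to 1 during skydive, dive or float states, though no shown node reads it.
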AnimBlueprint /BRCosmetics/Characters/Player/Female/Medium/Bodies/F_MED_RockClimber/Meshes/Parts/F_MED_Rock_Climber_FaceAcc_AnimBp
target_skeleton=&/F_MED_Rock_Climber_FaceAcc_Skeleton
graphs [
 AnimGraph {nodes [
   n1 = RigidBody(ComponentPose=Local To Component(LocalPose=Copy Pose From Mesh() {bUseAttachedParent=true})) {
    bClampLinearTranslationLimitToRefPose=true
    bOverrideWorldGravity=true
    overridePhysicsAsset=&/F_MED_Rock_Climber_FaceAcc_Physics
    simulationSpace=BaseBoneSpace
    alphaBoolBlend {blendOption=Linear}
    baseBoneRef {boneName=spine_05}
    componentAppliedLinearAccClamp=<6000 6000 6000>
    componentLinearAccScale=<0.25 0.25 0.25>
    componentLinearVelScale=<0.25 0.25 0.25>
    overrideWorldGravity {z=-980}
   }
   Output Pose(Result=Component To Local(ComponentPose=@n1)) {layerGroup=None}
  ]}
 EventGraph {nodes [
   Event BlueprintUpdateAnimation()
   n1 = Get bIsSkydiveDiveMode()
   n2 = Get bIsSkyDiving()
   n3 = Get bIsSkydiveFloating()
   > Set Skydive_Alpha(Skydive_Alpha:double=SelectFloat(A=1, B=0, bPickA=OR Boolean!(A=@n1, B=@n2, C:bool=@n3) -> (ReturnValue))) -> (Output_Get:double)
   TryGetPawnOwner()
  ]}
]
variables [Skydive_Alpha {type=number}]
